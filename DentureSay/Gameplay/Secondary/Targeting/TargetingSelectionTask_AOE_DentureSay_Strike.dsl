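; Targeting selection task for the DentureSay secondary strike AOE, supplying the trace's source offset, length and ignored actors.
; It casts the targeting request's source object to GA_DentureSay_Secondary_SpawnStrikes to read per-strike data.
; Trace length comes from a damage trace length scalable float, capped by the ability's ceiling clamp distance when positive.
; The same clamped length raises the source along Z, and the ability also provides the additional actors to ignore.
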(fn GetAdditionalActorsToIgnore
  (params (TargetingHandle :type Targeting%20Request%20Handle%20Structure%20%28by%20ref%29))
  (bind _sourceability (CallFunction|GetSourceAbility TargetingHandle))
  (Utilities|IsValid _sourceability
    (:Is%20Valid
      (return (Variables|/DentureSay/Gameplay/Secondary/GA_DentureSay_Secondary_SpawnStrikes.GA_DentureSay_Secondary_SpawnStrikes_C|GetActorsToIgnore _sourceability)))
    (:Is%20Not%20Valid
      (return 0))))

(fn GetCeilingClampedLength
  (params (TargetingHandle :type Targeting%20Request%20Handle%20Structure%20%28by%20ref%29))
  (bind _row_damagetracelength (Variables|Default|GetRow_DamageTraceLength))
  (bind _returnvalue (ScalableFloat|GetValueatLevel _row_damagetracelength))
  ;[0.15 0.1035 0.03299 0.5] Clamp to ceiling clamp distance from source ability if set
  (bind _sourceability (CallFunction|GetSourceAbility TargetingHandle))
  (Utilities|IsValid _sourceability
    (:Is%20Valid
      (if (> (Variables|/DentureSay/Gameplay/Secondary/GA_DentureSay_Secondary_SpawnStrikes.GA_DentureSay_Secondary_SpawnStrikes_C|GetCeilingClampDistance _sourceability) 0)
        (return (Math|Float|Min(Float) (Variables|/DentureSay/Gameplay/Secondary/GA_DentureSay_Secondary_SpawnStrikes.GA_DentureSay_Secondary_SpawnStrikes_C|GetCeilingClampDistance _sourceability) _returnvalue))
        (else
          _)))
    (:Is%20Not%20Valid
      _))
  (return _returnvalue))

(fn GetSourceAbility
  (params (TargetingHandle :type Targeting%20Request%20Handle%20Structure%20%28by%20ref%29))
  (bind (_sourceactor _instigatoractor _sourcelocation _sourcesocketname _sourceobject) (Utilities|Struct|BreakTargetingSourceContext (TargetingSystem|TargetingTypes|GetTargetingSourceContext (Targeting|GetTargetingSubsystem TargetingHandle) TargetingHandle)))
  (bind _returnvalue (Targeting|GetTargetingSubsystem TargetingHandle))
  (Utilities|IsValid _returnvalue
    (:Is%20Valid
      (bind _asga_denture_say_secondary_spawn_strikes (Utilities|Casting|CastToGA_DentureSay_Secondary_SpawnStrikes _sourceobject))
      (return _asga_denture_say_secondary_spawn_strikes))
    (:Is%20Not%20Valid
      (return 0))))

(fn GetSourceOffset
  (params (TargetingHandle :type Targeting%20Request%20Handle%20Structure%20%28by%20ref%29))
  (return 0.0 0.0 (CallFunction|GetCeilingClampedLength TargetingHandle)))

(fn GetTraceLength
  (params (TargetingHandle :type Targeting%20Request%20Handle%20Structure%20%28by%20ref%29))
  (return (CallFunction|GetCeilingClampedLength TargetingHandle)))
